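# Idle Niagara effect for the Relish Male pickaxe: flickering cyan-blue lightning and sparks along its Z axis.
# Lightning draws meshes, sprites and two lights per particle; a random normalized height drives placement, size and brightness, with light positions mirrored in XY.
# GPU SparkBurst throws a few speed-stretched sparks outward at random heights, falling under drag and gravity.
# Front-end modifiers boost lightning scale, flicker, lights and burst chance outside gameplay.
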NiagaraSystem /BRCosmetics/Weapons/FORT_Melee/Pickaxe_Relish_Male/FX/NS_Pickaxe_Relish_Male_Idle
dependencies {
 usedDataInterfaces [&NiagaraDataInterfaceCurve, &NiagaraDataInterfaceVectorCurve]
 usedDynamicInputs []
 usedModules [&SystemState]
}
emitters {
 Lightning {
  emitterUpdateScript [
   &EmitterState()
   &SpawnRate(Spawn Probability=0.5, SpawnRate=&UniformRangedFloat(Minimum=10, Maximum=30), Use Spawn Probability=true)
  ]
  particleSpawnScript [
   /SetVariables(Particles.NormalizedHeight:NiagaraFloat=&RandomRangeFloat(Minimum=0, Maximum=1)) {isSetParameters=true}
   /Niagara/Modules/Spawn/Initialization/InitializeParticle(
    Color=&MultiplyLinearColor(
     Linear Color=&RandomRangeLinearColor(Minimum=<rgba 0 5 3.8074 1>, Maximum=<rgba 0 4.9224 10 1>)
     Scale Factor=&MakeVector4_FromFloat(Float=&RandomRangeFloat(Minimum=0.04, Maximum=&Lerp_Float(A=1, B=0.25, Alpha=$Particles.NormalizedHeight)))
    )
    Lifetime=&RandomRangeFloat(Minimum=0.3, Maximum=0.8)
    Mesh Scale=&Multiply_VectorByFloat(
     Vector=&Lerp_Vector(
      A=&RandomRangeVector(Minimum=<0.025 0.01 0.01>, Maximum=<0.055 0.05 0.05>)
      B=&RandomRangeVector(Minimum=<0.01 0.02 0.02>, Maximum=<0.02 0.02 0.02>)
      Alpha=$Particles.NormalizedHeight
     )
     Float=&FN_FrontEndFloatModifier(Gameplay=1, Front End=1.3)
    )
    Module.Write Color=true
    Module.Write Lifetime=true
    Module.Write Mass=false
    Module.Write Material Random=true
    Module.Write Position=true
    Module.Write Scale=true
    Module.Write SpriteRotation=false
    Module.Write SpriteSize=true
    Position=&Lerp_Vector(A=<0 0 0>, B=<0 0 32>, Alpha=$Particles.NormalizedHeight)
    Sprite Size=&Vector2DFromFloat(Value=4)
   )
   /Niagara/Modules/Spawn/Orientation/V2/InitialMeshOrientation(
    Orientation Vector=<0 0 1>
    Rotation=&UniformRangedVector(Minimum=<-360 0 0>, Maximum=<360 0 0>)
    Use Orientation Vector=true
    Use Rotation Vector=true
   )
   &DynamicMaterialParameters(
    Index 0 Param 1=1
    Index 0 Param 2=&Lerp_Float(A=&RandomRangeFloat(Minimum=0.4, Maximum=0.75), B=0.2, Alpha=$Particles.NormalizedHeight)
    Index 0 Param 3=&RandomRangeFloat(Minimum=0, Maximum=1)
    Index 0 Param 4=&RandomRangeFloat(Minimum=0, Maximum=1)
    Module.Param0WriteEnabled=false
    Module.Param1WriteEnabled=true
    Module.Param2WriteEnabled=true
    Module.Param3WriteEnabled=true
   )
   &SystemLocation(MaskSpawnGroup=false, Offset=&Lerp_Vector(A=<0 0 10>, B=<0 0 45>, Alpha=$Particles.NormalizedHeight))
  ]
  particleUpdateScript [
   &ParticleState()
   &ScaleMeshSize(Scale Factor=&VectorFromCurve(VectorCurve=DataInterface() {propertyValues {
       XCurve {keys [
         {arriveTangent=1.4898, interpMode=RCIM_Cubic, leaveTangent=1.4898, tangentMode=RCTM_User, time=-0.0011, value=0.5}
         {arriveTangent=0.2007, interpMode=RCIM_Cubic, leaveTangent=0.2007, time=0.5, value=0.95}
         {arriveTangent=-1.4081, interpMode=RCIM_Cubic, leaveTangent=-1.4081, tangentMode=RCTM_User, time=0.9954, value=0.7}
        ]}
       YCurve {keys [<at 0 0.3>,<at 1 1>]}
       ZCurve {keys [<at 0 0.3>,<at 1 1>]}
      }}))
   &SolveForcesAndVelocity(Clamp Velocity=false, Limit Acceleration=false)
   &DynamicMaterialParameters(
    Index 0 Param 1=&FN_FrontEndFloatModifier(
     Gameplay=&RandomRangeFloat(Minimum=-1, Maximum=2, Evaluation Type=::NewEnumerator1)
     Front End=&RandomRangeFloat(Minimum=1.5, Maximum=2.5, Evaluation Type=::NewEnumerator1)
    )
    Module.Param0WriteEnabled=true
    Module.Param1WriteEnabled=false
    Module.Param2WriteEnabled=false
    Module.Param3WriteEnabled=false
   ) {module=DynamicMaterialParameters001}
   &FN_Light_v3(
    Color=$Particles.Color
    Create Color Attribute=true
    Create Exponent Attribute=true
    Create Radius Attribute=true
    Exponent=2
    Radius:NiagaraFloat=&Multiply_Float(
     A=&FN_FrontEndFloatModifier(Gameplay=0.4, Front End=1.3)
     B=/Niagara/DynamicInputs/Math/LerpMultipleFloats(
      Float Count:ENiagara_MultipleLerpCount=::NewEnumerator0
      1:NiagaraFloat=5
      2:NiagaraFloat=12
      3:NiagaraFloat=0
      4:NiagaraFloat=?
      5:NiagaraFloat=?
      Alpha:NiagaraFloat=$Particles.NormalizedAge
     )
    )
    Scale Color:NiagaraFloat=&FloatFromCurve(
     FloatCurve=DataInterface() {propertyValues {Curve {keys [<at 0 1>,<at 0.1656 0>,<at 0.2177 1>,<at 0.3102 0>,<at 0.4973 0.8>,<at 0.6017 0>,<at 0.7549 0.077>,<at 1 0>]}}}
     Scale Curve=&Multiply_Float(
      A=&FN_FrontEndFloatModifier(Gameplay=50, Front End=100)
      B=&ClampFloat(Float=&MakeFloatFromVector4(Vector 4=$Particles.DynamicMaterialParameter, Channel=::NewEnumerator0))
     )
    )
   )
   /SetVariables(Particles.PositionLightOffset:Vector3f=&Multiply_VectorByFloat(
     Vector=&Add_Vector(
      A=Linked() {linkedVariable {name=Particles.Position, type=NiagaraPosition}}
      B=&Add_Vector(
       A=&Multiply_VectorByFloat(
        Vector=<0 0 1>
        Float=&Multiply_Float(A=&MakeFloatFromVector(Vector=$Particles.Scale, Channel=::NewEnumerator0), B=400)
       )
       B=&Multiply_Vector(
        A=<1 1 0>
        B=&Add_Vector(
         A=&Lerp_Vector(A=<3 3 0>, B=<0 0 0>, Alpha=$Particles.NormalizedAge)
         B=&Multiply_VectorByFloat(
          Vector=&RandomVector(Vector Scale=5)
          Float=&LerpMultipleFloats(
           Float Count:ENiagara_MultipleLerpCount=::NewEnumerator0
           1:NiagaraFloat=0.5
           2:NiagaraFloat=1
           3:NiagaraFloat=0.5
           4:NiagaraFloat=?
           5:NiagaraFloat=?
           Alpha:NiagaraFloat=$Particles.NormalizedAge
          )
         )
        )
       )
      )
     )
     Float=&FN_FrontEndFloatModifier(Gameplay=1, Front End=1.3)
    )) {isSetParameters=true}
   /SetVariables(
    Particles.PositionLight01:Vector3f=$Particles.PositionLightOffset
    Particles.PositionLight02:Vector3f=&Multiply_Vector(A=$Particles.PositionLightOffset, B=<-1 -1 1>)
   ) {isSetParameters=true}
   &ScaleColor(
    Scale Alpha=&RandomRangeFloat(Minimum=-1, Maximum=1, Evaluation Type=::NewEnumerator1)
    ScaleA=true
    ScaleRGB=false
   )
   &Color(Color=&MakeCustomLinearColorFromBool(Bool=$User.bIsFrontEndPreview, True Color=<rgba 1 0 0 1>, False Color=<rgba 0 0 1 1>)) {enabled=false}
  ]
  rendererClasses [
   &NiagaraMeshRendererProperties
   &NiagaraLightRendererProperties
   &NiagaraSpriteRendererProperties
  ]
  renderers [
   {rendererClass=&NiagaraMeshRendererProperties, rendererIndex=0}
   {rendererClass=&NiagaraLightRendererProperties, rendererIndex=1}
   {rendererClass=&NiagaraLightRendererProperties, rendererIndex=2}
   {rendererClass=&NiagaraSpriteRendererProperties, rendererIndex=3}
   {rendererClass=&NiagaraSpriteRendererProperties, rendererIndex=4}
  ]
 }
 SparkBurst {
  simTarget=GPUComputeSim
  emitterUpdateScript [
   /SetVariables(Emitter.EmitterRandom:NiagaraFloat=&RandomRangeFloat(Minimum=20, Maximum=60)) {isSetParameters=true}
   &EmitterState()
   &SpawnBurst_Instantaneous(
    Spawn Count=&UniformRangedInt(Minimum=2, Maximum=3)
    Spawn Probability=&FN_FrontEndFloatModifier(Gameplay=0.5, Front End=1)
    Spawn Time=&RandomRangeFloat(Minimum=0, Maximum=1)
    Use Spawn Probability=true
   )
   &SpawnRate(SpawnRate=300, Use Spawn Probability=false) {enabled=false}
  ]
  particleSpawnScript [
   &InitializeParticle(
    Color=&MultiplyLinearColor(
     Linear Color=&RandomRangeLinearColor(Minimum=<rgba 0 30 27.355 30>, Maximum=<rgba 0 11.565 30 30>)
     Scale Factor=&MakeVector4_FromFloat(Float=&RandomRangeFloat(Minimum=0.05, Maximum=1))
    )
    Lifetime=&RandomRangeFloat(Minimum=0.1, Maximum=0.3)
    Module.Write Color=true
    Module.Write Lifetime=true
    Module.Write Mass=true
    Module.Write Material Random=true
    Module.Write Position=true
    Module.Write Scale=false
    Module.Write SpriteRotation=false
    Module.Write SpriteSize=true
    Position=&SimulationPosition()
    Sprite Size=&UniformRangedVector2D(Minimum=<0.35 0.25>, Maximum=<0.75 1>)
   )
   &SystemLocation(MaskSpawnGroup=false, Offset=&MakeVector(X=0, Y=0, Z=$Emitter.EmitterRandom))
   /Niagara/Modules/Spawn/Location/SphereLocation(
    Coordinate Space:ENiagaraCoordinateSpace=::Simulation
    MaskSpawnGroup=false
    Module.Override Randomness=false
    Sphere Radius=&UniformRangedFloat(Minimum=0, Maximum=5)
    Surface Only=false
   )
   &AddVelocity(Velocity=&Multiply_Vector(
     A=$Output.SphereLocation.SphereNormal
     B=&VectorFromFloat(Value=&UniformRangedFloat(Minimum=50, Maximum=200))
    )) {module=AddVelocity001}
  ]
  particleUpdateScript [
   &ParticleState()
   &Drag(Drag=&RandomRangeFloat(Minimum=1, Maximum=4), Use Linear Drag=true, Use Rotational Drag=false)
   &GravityForce(Gravity=&RandomRangeVector(Minimum=<0 0 -450>, Maximum=<1 1 -980>))
   &SolveForcesAndVelocity(Clamp Velocity=false, Limit Acceleration=false)
   &ScaleSpriteSizeBySpeed(Max Scale Factor=<0.75 8>, Min Scale Factor=<1 0.5>, Velocity Threshold=200)
   &ScaleSpriteSize(Scale Factor=&Vector2DFromFloat(Value=&FloatFromCurve(
      FloatCurve=DataInterface() {propertyValues {Curve {keys [
          {arriveTangent=43.4604, interpMode=RCIM_Cubic, leaveTangent=43.4604, tangentMode=RCTM_User, time=0.0001, value=0.0578}
          {arriveTangent=-0.0578, interpMode=RCIM_Cubic, leaveTangent=-0.0578, time=0.03, value=1}
          {arriveTangent=-3.2182, interpMode=RCIM_Cubic, leaveTangent=-3.2182, tangentMode=RCTM_User, time=1, value=0}
         ]}}}
      Scale Curve=1.75
     )))
   &ScaleColor(
    Scale Mode=::NewEnumerator1
    Scale RGBA=&MakeVector4_FromFloat(Float=&ClampFloat(Float=&Lerp_Float(A=2, B=0, Alpha=$Particles.NormalizedAge)))
    ScaleRGBA=true
   )
  ]
  rendererClasses [&NiagaraSpriteRendererProperties]
  renderers [{rendererClass=&NiagaraSpriteRendererProperties}]
 }
}
summary {
 systemName=NS_Pickaxe_Relish_Male_Idle
 emitters [
  {emitterName=SparkBurst, simTarget=GPUComputeSim, rendererClasses [&NiagaraSpriteRendererProperties]}
  {
   emitterName=Lightning
   rendererClasses [
    &NiagaraMeshRendererProperties
    &NiagaraLightRendererProperties
    &NiagaraSpriteRendererProperties
   ]
  }
 ]
 userVariables [
  User.bIsFrontEnd {type=NiagaraBool}
  User.bIsFrontEndPreview {type=NiagaraBool}
  User.ForwardDot {type=NiagaraFloat, defaultValue=0, description=""}
  User.Skeletal Mesh {
   type=NiagaraDataInterfaceSkeletalMesh
   defaultValue {
    struct=NiagaraExt_VariableValue_DataInterface
    value {dataInterface=&self:NiagaraDataInterfaceSkeletalMesh_0}
   }
  }
 ]
}
