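# Physics asset for the Relish Male two-handed pickaxe cosmetic, giving its handle and two small rope bones physics bodies.
# Each of the three bodies is a single capsule, with the rope capsules offset along and rotated to their bones.
# Both rope bodies constrain directly to the handle rather than to each other, each with its own narrow swing and twist limits for restrained dangling motion.
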PhysicsAsset /BRCosmetics/Weapons/FORT_Melee/Pickaxe_Relish_Male/Meshes/Relish_Male_PhysicsAsset
bodies {
 Melee_TwoHanded_Handle [{shapeType=Capsule, length=10, radius=1, shapeName=Melee_TwoHanded_Handle_capsule}]
 dyn_Rope_1 [{shapeType=Capsule, length=2, radius=1, shapeName=dyn_Rope_1_capsule, center=<1 0 0>, rotation=<rot 90 0 0>}]
 dyn_Rope_2 [{shapeType=Capsule, length=2, radius=1, shapeName=dyn_Rope_2_capsule, center=<1 0 0>, rotation=<rot 90 0 0>}]
}
constraints [
 {bone1Name=dyn_Rope_2, bone2Name=Melee_TwoHanded_Handle, swing1LimitDegrees=15, swing2LimitDegrees=25, twistLimitDegrees=23}
 {bone1Name=dyn_Rope_1, bone2Name=Melee_TwoHanded_Handle, swing1LimitDegrees=20, swing2LimitDegrees=20, twistLimitDegrees=15}
]
